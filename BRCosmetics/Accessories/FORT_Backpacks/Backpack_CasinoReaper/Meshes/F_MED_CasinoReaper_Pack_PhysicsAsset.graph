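# Physics asset for the CasinoReaper backpack cosmetic mesh (F_MED_CasinoReaper_Pack) in the Battle Royale accessories content.
# It has a single body on the root bone, a capsule offset from the bone and rotated to approximate the pack's shape.
# There are no constraints, so the pack is one rigid collision volume with no articulated parts.
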PhysicsAsset /BRCosmetics/Accessories/FORT_Backpacks/Backpack_CasinoReaper/Meshes/F_MED_CasinoReaper_Pack_PhysicsAsset
bodies {root [{shapeType=Capsule, length=11.2846, radius=11.2836, center=<0 -3.8334 3.5733>, rotation=<rot 0.7614 -90 -89.9741>}]}
constraints []
